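# Hair AnimBlueprint for the female GarageBand head cosmetic that adds secondary physics motion to hair and earrings.
# It copies pose and curves from the attached parent mesh, then runs a serial chain of partially blended Anim Dynamics nodes for back hair, front braids, ponytail and earrings.
# Back hair uses sphere collision against neck and spine planes, two ponytail bones use angular springs, and all nodes have LOD threshold 1.
# The event graph has no logic.
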AnimBlueprint /BRCosmetics/Characters/Player/Female/Medium/Heads/F_Med_GarageBand/Meshes/F_MED_GarageBand_Hair_AnimBP
target_skeleton=&/F_MED_GarageBand_Hair_Skeleton
graphs [
 AnimGraph {nodes [
   n1 = Anim Dynamics - Bone: dyn_hair_back_03(
    ComponentPose=Local To Component(LocalPose=Copy Pose From Mesh() {bCopyCurves=true, bUseAttachedParent=true})
    Alpha=0.5
    ExternalForce="0,0,0"
   ) {
    angularDampingOverride=0.988
    bGravityOverrideInSimSpace=true
    bOverrideAngularDamping=true
    lODThreshold=1
    numSolverIterationsPostUpdate=2
    numSolverIterationsPreUpdate=8
    simulationSpace=Actor
    alphaBoolBlend {blendOption=Linear}
    boundBone {boneName=dyn_hair_back_03}
    componentLinearAccScale=<0.2 0.2 0.2>
    physicsBodyDefinitions [{
      collisionType=CustomSphere
      sphereCollisionRadius=3
      boundBone {boneName=dyn_hair_back_03}
      boxExtents=<30 30 4>
      constraintSetup {angularTargetAxis=AxisZ, twistAxis=AxisZ, angularLimitsMax=<0 15 0>, angularLimitsMin=<0 -15 0>, angularTarget=<0 0 0>}
      localJointOffset=<0 0 -4>
     }]
    planarLimits [{drivingBone {boneName=neck_02}, planeTransform {location=<10 0 -5>, rotation=<rot 0 30.0001 -180>, scale=<0.5 0.5 0.5>}}]
   }
   n2 = Anim Dynamics - Bone: dyn_hair_back_01(ComponentPose=@n1, Alpha=0.7, ExternalForce="0,0,0") {
    angularDampingOverride=0.988
    bGravityOverrideInSimSpace=true
    bOverrideAngularDamping=true
    lODThreshold=1
    numSolverIterationsPostUpdate=2
    numSolverIterationsPreUpdate=8
    simulationSpace=Actor
    alphaBoolBlend {blendOption=Linear}
    boundBone {boneName=dyn_hair_back_01}
    componentLinearAccScale=<0.2 0.2 0.2>
    physicsBodyDefinitions [{
      collisionType=CustomSphere
      sphereCollisionRadius=3
      boundBone {boneName=dyn_hair_back_01}
      boxExtents=<30 30 4>
      constraintSetup {angularTargetAxis=AxisZ, twistAxis=AxisZ, angularLimitsMax=<15 15 0>, angularLimitsMin=<-15 -15 0>, angularTarget=<0 0 0>}
      localJointOffset=<0 0 -10>
     }]
    planarLimits [{drivingBone {boneName=spine_04}, planeTransform {location=<15 -3 0>, rotation=<rot 0 30.0001 -89.9998>, scale=<0.5 0.5 0.5>}}]
   }
   n3 = Anim Dynamics - Bone: dyn_hair_back_05(ComponentPose=@n2, Alpha=0.5, ExternalForce="0,0,0") {
    angularDampingOverride=0.988
    bGravityOverrideInSimSpace=true
    bOverrideAngularDamping=true
    lODThreshold=1
    numSolverIterationsPostUpdate=2
    numSolverIterationsPreUpdate=8
    simulationSpace=Actor
    alphaBoolBlend {blendOption=Linear}
    boundBone {boneName=dyn_hair_back_05}
    componentLinearAccScale=<0.2 0.2 0.2>
    physicsBodyDefinitions [{
      collisionType=CustomSphere
      sphereCollisionRadius=3
      boundBone {boneName=dyn_hair_back_05}
      boxExtents=<30 30 4>
      constraintSetup {angularTargetAxis=AxisZ, twistAxis=AxisZ, angularLimitsMax=<0 15 0>, angularLimitsMin=<0 -15 0>, angularTarget=<0 0 0>}
      localJointOffset=<0 0 -4>
     }]
    planarLimits [{drivingBone {boneName=neck_02}, planeTransform {location=<10 0 5>, rotation=<rot 0 30.0001 0>, scale=<0.5 0.5 0.5>}}]
   }
   n4 = Anim Dynamics - Bone: dyn_braid_front_10(ComponentPose=@n3, Alpha=0.7, ExternalForce="0,0,0") {
    angularDampingOverride=0.9
    bGravityOverrideInSimSpace=true
    bOverrideAngularDamping=true
    bUsePlanarLimit=false
    lODThreshold=1
    simulationSpace=Actor
    alphaBoolBlend {blendOption=Linear}
    boundBone {boneName=dyn_braid_front_10}
    componentLinearAccScale=<1 1 1>
    physicsBodyDefinitions [{
      sphereCollisionRadius=0
      boundBone {boneName=dyn_braid_front_10}
      boxExtents=<5 5 2.5>
      constraintSetup {angularTargetAxis=AxisZ, twistAxis=AxisZ, angularLimitsMax=<15 15 0>, angularLimitsMin=<-15 -15 0>, angularTarget=<0 0 0>}
      localJointOffset=<0 0 -0.1>
     }]
   }
   n5 = Anim Dynamics - Bone: dyn_braid_front_03(ComponentPose=@n4, Alpha=0.7, ExternalForce="0,0,0") {
    angularDampingOverride=0.9
    bGravityOverrideInSimSpace=true
    bOverrideAngularDamping=true
    bUsePlanarLimit=false
    lODThreshold=1
    simulationSpace=Actor
    alphaBoolBlend {blendOption=Linear}
    boundBone {boneName=dyn_braid_front_03}
    componentLinearAccScale=<1 1 1>
    physicsBodyDefinitions [{
      sphereCollisionRadius=0
      boundBone {boneName=dyn_braid_front_03}
      boxExtents=<5 5 2.5>
      constraintSetup {angularTargetAxis=AxisZ, twistAxis=AxisZ, angularLimitsMax=<15 15 0>, angularLimitsMin=<-15 -15 0>, angularTarget=<0 0 0>}
      localJointOffset=<0 0 -0.1>
     }]
   }
   n6 = Anim Dynamics - Bone: dyn_braid_front_01(ComponentPose=@n5, Alpha=0.7, ExternalForce="0,0,0") {
    angularDampingOverride=0.9
    bGravityOverrideInSimSpace=true
    bOverrideAngularDamping=true
    bUsePlanarLimit=false
    lODThreshold=1
    simulationSpace=Actor
    alphaBoolBlend {blendOption=Linear}
    boundBone {boneName=dyn_braid_front_01}
    componentLinearAccScale=<1 1 1>
    physicsBodyDefinitions [{
      sphereCollisionRadius=0
      boundBone {boneName=dyn_braid_front_01}
      boxExtents=<5 5 2.5>
      constraintSetup {angularTargetAxis=AxisZ, twistAxis=AxisZ, angularLimitsMax=<15 15 0>, angularLimitsMin=<-15 -15 0>, angularTarget=<0 0 0>}
      localJointOffset=<0 0 -0.1>
     }]
   }
   n7 = Anim Dynamics - Bone: dyn_ponyTail_01(ComponentPose=@n6, Alpha=0.75, ExternalForce="0,0,0") {
    angularDampingOverride=0.99
    angularSpringConstant=500
    bAngularSpring=true
    bGravityOverrideInSimSpace=true
    bOverrideAngularDamping=true
    gravityScale=0.35
    lODThreshold=1
    numSolverIterationsPostUpdate=2
    numSolverIterationsPreUpdate=8
    alphaBoolBlend {blendOption=Linear}
    boundBone {boneName=dyn_ponyTail_01}
    componentLinearAccScale=<0.1 0.1 0.1>
    physicsBodyDefinitions [{
      sphereCollisionRadius=0
      boundBone {boneName=dyn_ponyTail_01}
      boxExtents=<5 5 5>
      constraintSetup {angularTargetAxis=AxisZ, twistAxis=AxisZ, angularLimitsMax=<10 20 40>, angularLimitsMin=<-30 -20 -40>, angularTarget=<0 0 0>}
      localJointOffset=<0 0 -2>
     }]
   }
   n8 = Anim Dynamics - Bone: dyn_ponyTail_03(ComponentPose=@n7, Alpha=0.75, ExternalForce="0,0,0") {
    angularDampingOverride=0.99
    angularSpringConstant=500
    bAngularSpring=true
    bGravityOverrideInSimSpace=true
    bOverrideAngularDamping=true
    gravityScale=0.35
    lODThreshold=1
    numSolverIterationsPostUpdate=2
    numSolverIterationsPreUpdate=8
    alphaBoolBlend {blendOption=Linear}
    boundBone {boneName=dyn_ponyTail_03}
    componentLinearAccScale=<0.1 0.1 0.1>
    physicsBodyDefinitions [{
      sphereCollisionRadius=0
      boundBone {boneName=dyn_ponyTail_03}
      boxExtents=<5 5 5>
      constraintSetup {angularTargetAxis=AxisZ, twistAxis=AxisZ, angularLimitsMax=<40 20 40>, angularLimitsMin=<-40 -20 -40>, angularTarget=<0 0 0>}
      localJointOffset=<0 0 -2>
     }]
   }
   n9 = Anim Dynamics - Bone: dyn_ponyTail_11(ComponentPose=@n8, Alpha=1, ExternalForce="0,0,0") {
    bGravityOverrideInSimSpace=true
    bUsePlanarLimit=false
    gravityScale=0.35
    lODThreshold=1
    numSolverIterationsPostUpdate=2
    numSolverIterationsPreUpdate=8
    simulationSpace=Actor
    alphaBoolBlend {blendOption=Linear}
    boundBone {boneName=dyn_ponyTail_11}
    componentLinearAccScale=<1 1 1>
    physicsBodyDefinitions [{
      sphereCollisionRadius=0
      boundBone {boneName=dyn_ponyTail_11}
      boxExtents=<5 5 2.5>
      constraintSetup {angularTargetAxis=AxisZ, twistAxis=AxisZ, angularLimitsMax=<15 15 0>, angularLimitsMin=<-15 -15 0>, angularTarget=<0 0 0>}
      localJointOffset=<0 0 -0.1>
     }]
   }
   n10 = Anim Dynamics - Bone: dyn_earring_01(ComponentPose=@n9, Alpha=0.3, ExternalForce="0,0,0") {
    alpha=0.9
    angularDampingOverride=0.94
    angularSpringConstant=500
    bGravityOverrideInSimSpace=true
    lODThreshold=1
    numSolverIterationsPostUpdate=4
    numSolverIterationsPreUpdate=16
    alphaBoolBlend {blendOption=Linear}
    boundBone {boneName=dyn_earring_01}
    componentLinearAccScale=<1 1 1>
    componentLinearVelScale=<1 1 1>
    physicsBodyDefinitions [{
      sphereCollisionRadius=0
      boundBone {boneName=dyn_earring_01}
      boxExtents=<15 15 5>
      constraintSetup {angularTargetAxis=AxisZ, twistAxis=AxisZ, angularLimitsMax=<180 30 80>, angularLimitsMin=<-180 -30 -80>, angularTarget=<0 0 1>}
      localJointOffset=<0 0 -1>
     }]
   }
   n11 = Anim Dynamics - Bone: dyn_earring_02(ComponentPose=@n10, Alpha=0.3, ExternalForce="0,0,0") {
    alpha=0.9
    angularDampingOverride=0.95
    angularSpringConstant=500
    bGravityOverrideInSimSpace=true
    lODThreshold=1
    numSolverIterationsPostUpdate=4
    numSolverIterationsPreUpdate=16
    alphaBoolBlend {blendOption=Linear}
    boundBone {boneName=dyn_earring_02}
    componentLinearAccScale=<1 1 1>
    physicsBodyDefinitions [{
      sphereCollisionRadius=0
      boundBone {boneName=dyn_earring_02}
      boxExtents=<15 15 5>
      constraintSetup {angularTargetAxis=AxisZ, twistAxis=AxisZ, angularLimitsMax=<180 30 80>, angularLimitsMin=<-180 -30 -80>, angularTarget=<0 0 1>}
      localJointOffset=<0 0 -1>
     }]
   }
   Output Pose(Result=Component To Local(ComponentPose=@n11)) {layerGroup=None}
  ]}
 EventGraph {nodes [
   TryGetPawnOwner()
   Event BlueprintUpdateAnimation()
  ]}
]
variables []
